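# Physics asset for the HornetBerm medium male player body cosmetic, covering its dynamic cape and sheath.
# The cape is three six-bone capsule chains (center, left, right) hung from MSTR bones, with low twist limits and swing widening below the top joints.
# Angularly free constraints tie the side chains' top bones to the center's, while the sheath is one limited-swing capsule.
# Capsules on spine, pelvis, shoulders, upper arms, legs and feet approximate the body.
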PhysicsAsset /BRCosmetics/Characters/Player/Male/Medium/Bodies/M_MED_HornetBerm/Meshes/M_MED_HornetBerm_CapePhysics
bodies {
 calf_l [{shapeType=Capsule, length=32, radius=8, center=<-24 -2 0>, rotation=<rot 0 90 90>}]
 calf_r [{shapeType=Capsule, length=32, radius=8, center=<24 2 0>, rotation=<rot 0 90 90>}]
 clavicle_l [{shapeType=Capsule, length=12, radius=10, shapeName=clavicle_l_capsule, center=<8 0 0>, rotation=<rot 0 90 85>}]
 clavicle_r [{shapeType=Capsule, length=12, radius=10, shapeName=clavicle_r_capsule, center=<-8 0 0>, rotation=<rot 0 90 85>}]
 dyn_cape_01 [{shapeType=Capsule, length=11.7556, radius=5.9773, shapeName=dyn_cape_01_capsule, center=<7.15 0 8>, rotation=<rot 90 0 0>}]
 dyn_cape_02 [{shapeType=Capsule, length=9.602, radius=8.1308, shapeName=dyn_cape_01_capsule, center=<7.15 0 5>, rotation=<rot 90 0 0>}]
 dyn_cape_03 [{shapeType=Capsule, length=9.602, radius=8.1308, shapeName=dyn_cape_01_capsule, center=<7.15 0 5>, rotation=<rot 90 0 0>}]
 dyn_cape_04 [{shapeType=Capsule, length=9.602, radius=8.1308, shapeName=dyn_cape_01_capsule, center=<7.15 0 5>, rotation=<rot 90 0 0>}]
 dyn_cape_05 [{shapeType=Capsule, length=9.602, radius=8.1308, shapeName=dyn_cape_01_capsule, center=<7.15 0 5>, rotation=<rot 90 0 0>}]
 dyn_cape_06 [{shapeType=Capsule, length=9.602, radius=8.1308, shapeName=dyn_cape_01_capsule, center=<7.15 0 5>, rotation=<rot 90 0 0>}]
 dyn_cape_MSTR [{shapeType=Capsule, length=0.505, radius=0.505, shapeName=dyn_cape_r_MSTR_capsule, rotation=<rot 0 -90 0>}]
 dyn_cape_l_01 [{shapeType=Capsule, length=11.7556, radius=5.9773, shapeName=dyn_cape_01_capsule, center=<7.15 0 8>, rotation=<rot 90 0 0>}]
 dyn_cape_l_02 [{shapeType=Capsule, length=9.602, radius=8.1308, shapeName=dyn_cape_01_capsule, center=<7.15 0 5>, rotation=<rot 90 0 0>}]
 dyn_cape_l_03 [{shapeType=Capsule, length=9.602, radius=8.1308, shapeName=dyn_cape_01_capsule, center=<7.15 0 5>, rotation=<rot 90 0 0>}]
 dyn_cape_l_04 [{shapeType=Capsule, length=9.602, radius=8.1308, shapeName=dyn_cape_01_capsule, center=<7.15 0 5>, rotation=<rot 90 0 0>}]
 dyn_cape_l_05 [{shapeType=Capsule, length=9.602, radius=8.1308, shapeName=dyn_cape_01_capsule, center=<7.15 0 5>, rotation=<rot 90 0 0>}]
 dyn_cape_l_06 [{shapeType=Capsule, length=9.602, radius=8.1308, shapeName=dyn_cape_01_capsule, center=<7.15 0 5>, rotation=<rot 90 0 0>}]
 dyn_cape_l_MSTR [{shapeType=Capsule, length=0.505, radius=0.505, shapeName=dyn_cape_r_MSTR_capsule, rotation=<rot 0 -90 0>}]
 dyn_cape_r_01 [{shapeType=Capsule, length=11.7556, radius=5.9773, shapeName=dyn_cape_01_capsule, center=<7.15 0 8>, rotation=<rot 90 0 0>}]
 dyn_cape_r_02 [{shapeType=Capsule, length=9.602, radius=8.1308, shapeName=dyn_cape_01_capsule, center=<7.15 0 5>, rotation=<rot 90 0 0>}]
 dyn_cape_r_03 [{shapeType=Capsule, length=9.602, radius=8.1308, shapeName=dyn_cape_01_capsule, center=<7.15 0 5>, rotation=<rot 90 0 0>}]
 dyn_cape_r_04 [{shapeType=Capsule, length=9.602, radius=8.1308, shapeName=dyn_cape_01_capsule, center=<7.15 0 5>, rotation=<rot 90 0 0>}]
 dyn_cape_r_05 [{shapeType=Capsule, length=9.602, radius=8.1308, shapeName=dyn_cape_01_capsule, center=<7.15 0 5>, rotation=<rot 90 0 0>}]
 dyn_cape_r_06 [{shapeType=Capsule, length=9.602, radius=8.1308, shapeName=dyn_cape_01_capsule, center=<7.15 0 5>, rotation=<rot 90 0 0>}]
 dyn_cape_r_MSTR [{shapeType=Capsule, length=0.505, radius=0.505, shapeName=dyn_cape_r_MSTR_capsule, rotation=<rot 0 -90 0>}]
 dyn_sheath_01 [{shapeType=Capsule, length=65, radius=3, center=<1 0 -15>, rotation=<rot 1 0 0>}]
 dyn_sheath_MSTR [{shapeType=Capsule, length=0.505, radius=0.505, shapeName=dyn_sheath_MSTR_capsule, rotation=<rot 0 -90 0>}]
 foot_l [{shapeType=Capsule, length=15, radius=9, center=<-3 8 0>, rotation=<rot 0 0 90>}]
 foot_r [{shapeType=Capsule, length=15, radius=9, center=<3 -8 0>, rotation=<rot 0 0 90>}]
 pelvis [{shapeType=Capsule, length=12, radius=12, center=<0 -1 0>}]
 spine_02 [{shapeType=Capsule, length=8, radius=11, shapeName=spine_02_capsule}]
 spine_03 [{shapeType=Capsule, length=10, radius=11, shapeName=spine_03_capsule}]
 spine_04 [{shapeType=Capsule, length=10, radius=12, shapeName=spine_04_capsule}]
 spine_05 [{shapeType=Capsule, length=10, radius=13, shapeName=spine_05_capsule}]
 thigh_l [{shapeType=Capsule, length=32, radius=11, center=<-21 0 0>, rotation=<rot 0 90 90>}]
 thigh_r [{shapeType=Capsule, length=32, radius=11, center=<21 0 0>, rotation=<rot 0 90 90>}]
 upperarm_l [{shapeType=Capsule, length=19, radius=7, shapeName=upperarm_l_capsule, center=<14 -1 0>, rotation=<rot 0 90 90>}]
 upperarm_r [{shapeType=Capsule, length=19, radius=7, shapeName=upperarm_r_capsule, center=<-14 1 0>, rotation=<rot 0 90 90>}]
}
constraints [
 {bone1Name=dyn_sheath_01, bone2Name=dyn_sheath_MSTR, swing1LimitDegrees=25, swing2LimitDegrees=25, twistLimitDegrees=1}
 {bone1Name=dyn_cape_01, bone2Name=dyn_cape_MSTR, swing1LimitDegrees=10, swing2LimitDegrees=45, twistLimitDegrees=5}
 {bone1Name=dyn_cape_02, bone2Name=dyn_cape_01, swing1LimitDegrees=20, swing2LimitDegrees=25, twistLimitDegrees=5}
 {bone1Name=dyn_cape_03, bone2Name=dyn_cape_02, swing1LimitDegrees=45, swing2LimitDegrees=45, twistLimitDegrees=5}
 {bone1Name=dyn_cape_04, bone2Name=dyn_cape_03, swing1LimitDegrees=45, swing2LimitDegrees=45, twistLimitDegrees=5}
 {bone1Name=dyn_cape_05, bone2Name=dyn_cape_04, swing1LimitDegrees=45, swing2LimitDegrees=45, twistLimitDegrees=5}
 {bone1Name=dyn_cape_06, bone2Name=dyn_cape_05, swing1LimitDegrees=45, swing2LimitDegrees=45, twistLimitDegrees=5}
 {bone1Name=dyn_cape_l_01, bone2Name=dyn_cape_l_MSTR, swing1LimitDegrees=10, swing2LimitDegrees=45, twistLimitDegrees=5}
 {bone1Name=dyn_cape_l_02, bone2Name=dyn_cape_l_01, swing1LimitDegrees=20, swing2LimitDegrees=25, twistLimitDegrees=5}
 {bone1Name=dyn_cape_l_03, bone2Name=dyn_cape_l_02, swing1LimitDegrees=45, swing2LimitDegrees=45, twistLimitDegrees=5}
 {bone1Name=dyn_cape_l_04, bone2Name=dyn_cape_l_03, swing1LimitDegrees=45, swing2LimitDegrees=45, twistLimitDegrees=5}
 {bone1Name=dyn_cape_l_05, bone2Name=dyn_cape_l_04, swing1LimitDegrees=45, swing2LimitDegrees=45, twistLimitDegrees=5}
 {bone1Name=dyn_cape_l_06, bone2Name=dyn_cape_l_05, swing1LimitDegrees=45, swing2LimitDegrees=45, twistLimitDegrees=5}
 {bone1Name=dyn_cape_r_01, bone2Name=dyn_cape_r_MSTR, swing1LimitDegrees=10, swing2LimitDegrees=45, twistLimitDegrees=5}
 {bone1Name=dyn_cape_r_02, bone2Name=dyn_cape_r_01, swing1LimitDegrees=20, swing2LimitDegrees=25, twistLimitDegrees=5}
 {bone1Name=dyn_cape_r_03, bone2Name=dyn_cape_r_02, swing1LimitDegrees=45, swing2LimitDegrees=45, twistLimitDegrees=5}
 {bone1Name=dyn_cape_r_04, bone2Name=dyn_cape_r_03, swing1LimitDegrees=45, swing2LimitDegrees=45, twistLimitDegrees=5}
 {bone1Name=dyn_cape_r_05, bone2Name=dyn_cape_r_04, swing1LimitDegrees=45, swing2LimitDegrees=45, twistLimitDegrees=5}
 {bone1Name=dyn_cape_r_06, bone2Name=dyn_cape_r_05, swing1LimitDegrees=45, swing2LimitDegrees=45, twistLimitDegrees=5}
 {
  bone1Name=dyn_cape_l_01
  bone2Name=dyn_cape_01
  swing1LimitDegrees=45
  swing1Motion=Free
  swing2LimitDegrees=45
  swing2Motion=Free
  twistLimitDegrees=45
  twistMotion=Free
 }
 {
  bone1Name=dyn_cape_r_01
  bone2Name=dyn_cape_01
  swing1LimitDegrees=45
  swing1Motion=Free
  swing2LimitDegrees=45
  swing2Motion=Free
  twistLimitDegrees=45
  twistMotion=Free
 }
]
