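# Physics asset for the M_MED_HornetBerm male player body cosmetic, setting up secondary motion for its tassels, belt and sheath.
# Capsules on the pelvis, spine, legs and feet form the body volume around short capsule chains on the dyn_ bones.
# Each chain has angular limits and starts at a tiny MSTR anchor, free-rotating on a thigh for tassels or spine_02 for the belt.
# The sheath's anchor has no constraint to the body.
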PhysicsAsset /BRCosmetics/Characters/Player/Male/Medium/Bodies/M_MED_HornetBerm/Meshes/M_MED_HornetBerm_Physics
bodies {
 calf_l [{shapeType=Capsule, length=32, radius=8, center=<-24 -2 0>, rotation=<rot 0 90 90>}]
 calf_r [{shapeType=Capsule, length=32, radius=8, center=<24 2 0>, rotation=<rot 0 90 90>}]
 dyn_belt_01 [{shapeType=Capsule, length=3, radius=3, shapeName=dyn_belt_01_capsule, center=<4 0 2>, rotation=<rot 0 0 90>}]
 dyn_belt_02 [{shapeType=Capsule, length=3, radius=3, shapeName=dyn_belt_02_capsule, center=<4 0 2>, rotation=<rot 0 0 90>}]
 dyn_belt_03 [{shapeType=Capsule, length=3, radius=3, shapeName=dyn_belt_03_capsule, center=<4 0 2>, rotation=<rot 0 0 90>}]
 dyn_belt_MSTR [{shapeType=Capsule, length=0.505, radius=0.505, shapeName=dyn_belt_MSTR_capsule, rotation=<rot 0 -90 0>}]
 dyn_bk_r_01_tassle_01 [{shapeType=Capsule, length=3, radius=2, shapeName=dyn_bk_r_01_tassle_01_capsule, center=<2 -1.5 0>, rotation=<rot 0 90 90>}]
 dyn_bk_r_01_tassle_MSTR [{shapeType=Capsule, length=0.505, radius=0.505, shapeName=dyn_bk_r_01_tassle_MSTR_capsule, rotation=<rot 0 -90 0>}]
 dyn_bk_r_02_tassle_01 [{shapeType=Capsule, length=3, radius=2, shapeName=dyn_bk_r_02_tassle_01_capsule, center=<2 -1.5 0>, rotation=<rot 0 90 90>}]
 dyn_bk_r_02_tassle_02 [{shapeType=Capsule, length=3, radius=2, shapeName=dyn_bk_r_02_tassle_02_capsule, center=<2 -1.5 0>, rotation=<rot 0 90 90>}]
 dyn_bk_r_02_tassle_MSTR [{shapeType=Capsule, length=0.505, radius=0.505, shapeName=dyn_bk_r_02_tassle_MSTR_capsule, rotation=<rot 0 -90 0>}]
 dyn_ft_l_01_tassle_01 [{shapeType=Capsule, length=3, radius=2, shapeName=dyn_ft_l_01_tassle_01_capsule, center=<2 -1.25 0>, rotation=<rot 0 90 90>}]
 dyn_ft_l_01_tassle_02 [{shapeType=Capsule, length=3, radius=2, shapeName=dyn_ft_l_01_tassle_02_capsule, center=<2 -1.25 0>, rotation=<rot 0 90 90>}]
 dyn_ft_l_01_tassle_03 [{shapeType=Capsule, length=3, radius=2, shapeName=dyn_ft_l_01_tassle_03_capsule, center=<2 -1.25 0>, rotation=<rot 0 90 160>}]
 dyn_ft_l_01_tassle_04 [{shapeType=Capsule, length=3, radius=2, shapeName=dyn_ft_l_01_tassle_04_capsule, center=<2 -1.25 0>, rotation=<rot 0 90 -20>}]
 dyn_ft_l_01_tassle_MSTR [{shapeType=Capsule, length=0.505, radius=0.505, shapeName=dyn_ft_l_01_tassle_MSTR_capsule, rotation=<rot 0 -90 0>}]
 dyn_ft_l_02_tassle_01 [{shapeType=Capsule, length=3, radius=2, shapeName=dyn_ft_l_02_tassle_01_capsule, center=<2 -1 0>, rotation=<rot 0 90 90>}]
 dyn_ft_l_02_tassle_02 [{shapeType=Capsule, length=3, radius=2, shapeName=dyn_ft_l_02_tassle_02_capsule, center=<2 -1 0>, rotation=<rot 0 90 90>}]
 dyn_ft_l_02_tassle_03 [{shapeType=Capsule, length=3, radius=2, shapeName=dyn_ft_l_02_tassle_03_capsule, center=<2 -1 0>, rotation=<rot 0 90 90>}]
 dyn_ft_l_02_tassle_04 [{shapeType=Capsule, length=3, radius=2, shapeName=dyn_ft_l_02_tassle_04_capsule, center=<2 -1 0>, rotation=<rot 0 90 90>}]
 dyn_ft_l_02_tassle_MSTR [{shapeType=Capsule, length=0.505, radius=0.505, shapeName=dyn_ft_l_02_tassle_MSTR_capsule, rotation=<rot 0 -90 0>}]
 dyn_ft_l_03_tassle_01 [{shapeType=Capsule, length=3, radius=2, shapeName=dyn_ft_l_03_tassle_01_capsule, center=<2 -1.5 0>, rotation=<rot 0 90 90>}]
 dyn_ft_l_03_tassle_02 [{shapeType=Capsule, length=3, radius=2, shapeName=dyn_ft_l_03_tassle_02_capsule, center=<2 -1.5 0>, rotation=<rot 0 90 90>}]
 dyn_ft_l_03_tassle_03 [{shapeType=Capsule, length=3, radius=2, shapeName=dyn_ft_l_03_tassle_03_capsule, center=<2 -1.5 0>, rotation=<rot 0 90 90>}]
 dyn_ft_l_03_tassle_MSTR [{shapeType=Capsule, length=0.505, radius=0.505, shapeName=dyn_ft_l_03_tassle_MSTR_capsule, rotation=<rot 0 -90 0>}]
 dyn_ft_r_01_tassle_01 [{shapeType=Capsule, length=3, radius=2, shapeName=dyn_ft_r_01_tassle_01_capsule, center=<2 -1.25 0>, rotation=<rot 0 90 90>}]
 dyn_ft_r_01_tassle_02 [{shapeType=Capsule, length=3, radius=2, shapeName=dyn_ft_r_01_tassle_02_capsule, center=<2 -1.25 0>, rotation=<rot 0 90 90>}]
 dyn_ft_r_01_tassle_03 [{shapeType=Capsule, length=3, radius=2, shapeName=dyn_ft_r_01_tassle_03_capsule, center=<2 -1.25 0>, rotation=<rot 0 90 90>}]
 dyn_ft_r_01_tassle_MSTR [{shapeType=Capsule, length=0.505, radius=0.505, shapeName=dyn_ft_r_01_tassle_MSTR_capsule, rotation=<rot 0 -90 0>}]
 dyn_ft_r_02_tassle_01 [{shapeType=Capsule, length=3, radius=2, shapeName=dyn_ft_r_02_tassle_01_capsule, center=<2 -1 0>, rotation=<rot 0 90 90>}]
 dyn_ft_r_02_tassle_02 [{shapeType=Capsule, length=3, radius=2, shapeName=dyn_ft_r_02_tassle_02_capsule, center=<2 -1 0>, rotation=<rot 0 90 90>}]
 dyn_ft_r_02_tassle_03 [{shapeType=Capsule, length=3, radius=2, shapeName=dyn_ft_r_02_tassle_03_capsule, center=<2 -1 0>, rotation=<rot 0 90 90>}]
 dyn_ft_r_02_tassle_MSTR [{shapeType=Capsule, length=0.505, radius=0.505, shapeName=dyn_ft_r_02_tassle_MSTR_capsule, rotation=<rot 0 -90 0>}]
 dyn_ft_r_03_tassle_01 [{shapeType=Capsule, length=3, radius=2, shapeName=dyn_ft_r_03_tassle_01_capsule, center=<2 -1.25 0>, rotation=<rot 0 90 90>}]
 dyn_ft_r_03_tassle_02 [{shapeType=Capsule, length=3, radius=2, shapeName=dyn_ft_r_03_tassle_02_capsule, center=<2 -1.25 0>, rotation=<rot 0 90 90>}]
 dyn_ft_r_03_tassle_03 [{shapeType=Capsule, length=3, radius=2, shapeName=dyn_ft_r_03_tassle_03_capsule, center=<2 -1.25 0>, rotation=<rot 0 90 90>}]
 dyn_ft_r_03_tassle_MSTR [{shapeType=Capsule, length=0.505, radius=0.505, shapeName=dyn_ft_r_03_tassle_MSTR_capsule, rotation=<rot 0 -90 0>}]
 dyn_l_tassle_01 [{shapeType=Capsule, length=3, radius=2, shapeName=dyn_l_tassle_01_capsule, center=<2 -1.5 0>, rotation=<rot 0 90 90>}]
 dyn_l_tassle_02 [{shapeType=Capsule, length=3, radius=2, shapeName=dyn_l_tassle_02_capsule, center=<2 -1.5 0>, rotation=<rot 0 90 90>}]
 dyn_l_tassle_03 [{shapeType=Capsule, length=3, radius=2, shapeName=dyn_l_tassle_03_capsule, center=<2 -1.5 0>, rotation=<rot 0 90 90>}]
 dyn_l_tassle_MSTR [{shapeType=Capsule, length=0.505, radius=0.505, shapeName=dyn_l_tassle_MSTR_capsule, rotation=<rot 0 -90 0>}]
 dyn_r_tassle_01 [{shapeType=Capsule, length=3, radius=2, shapeName=dyn_r_tassle_01_capsule, center=<2 -1.5 0>, rotation=<rot 0 90 90>}]
 dyn_r_tassle_02 [{shapeType=Capsule, length=3, radius=2, shapeName=dyn_r_tassle_02_capsule, center=<2 -1.5 0>, rotation=<rot 0 90 90>}]
 dyn_r_tassle_03 [{shapeType=Capsule, length=3, radius=2, shapeName=dyn_r_tassle_03_capsule, center=<2 -1.5 0>, rotation=<rot 0 90 90>}]
 dyn_r_tassle_MSTR [{shapeType=Capsule, length=0.505, radius=0.505, shapeName=dyn_r_tassle_MSTR_capsule, rotation=<rot 0 -90 0>}]
 dyn_sheath_01 [{shapeType=Capsule, length=65, radius=3, center=<1 0 -15>, rotation=<rot 1 0 0>}]
 dyn_sheath_MSTR [{shapeType=Capsule, length=0.505, radius=0.505, shapeName=dyn_sheath_MSTR_capsule, rotation=<rot 0 -90 0>}]
 foot_l [{shapeType=Capsule, length=15, radius=9, center=<-3 8 0>, rotation=<rot 0 0 90>}]
 foot_r [{shapeType=Capsule, length=15, radius=9, center=<3 -8 0>, rotation=<rot 0 0 90>}]
 pelvis [{shapeType=Capsule, length=12, radius=11, center=<0 -1 0>}]
 spine_02 [{shapeType=Capsule, length=8, radius=10, shapeName=spine_02_capsule, center=<3 0 0>}]
 spine_03 [{shapeType=Capsule, length=10, radius=11, center=<5 1 0>}]
 thigh_l [{shapeType=Capsule, length=32, radius=9, center=<-21 -1 0>, rotation=<rot 0 87 90>}]
 thigh_r [{shapeType=Capsule, length=32, radius=9, center=<21 1 0>, rotation=<rot 0 87 90>}]
}
constraints [
 {bone1Name=dyn_sheath_01, bone2Name=dyn_sheath_MSTR, swing1LimitDegrees=5, swing2LimitDegrees=25, twistLimitDegrees=0}
 {
  bone1Name=dyn_ft_l_01_tassle_MSTR
  bone2Name=thigh_l
  swing1LimitDegrees=45
  swing1Motion=Free
  swing2LimitDegrees=45
  swing2Motion=Free
  twistLimitDegrees=45
  twistMotion=Free
 }
 {bone1Name=dyn_ft_l_01_tassle_01, bone2Name=dyn_ft_l_01_tassle_MSTR, swing1LimitDegrees=25, swing2LimitDegrees=5, twistLimitDegrees=1}
 {bone1Name=dyn_ft_l_01_tassle_02, bone2Name=dyn_ft_l_01_tassle_01, swing1LimitDegrees=25, swing2LimitDegrees=5, twistLimitDegrees=1}
 {bone1Name=dyn_ft_l_01_tassle_03, bone2Name=dyn_ft_l_01_tassle_02, swing1LimitDegrees=25, swing2LimitDegrees=5, twistLimitDegrees=1}
 {bone1Name=dyn_ft_l_01_tassle_04, bone2Name=dyn_ft_l_01_tassle_03, swing1LimitDegrees=25, swing2LimitDegrees=5, twistLimitDegrees=1}
 {
  bone1Name=dyn_ft_l_02_tassle_MSTR
  bone2Name=thigh_l
  swing1LimitDegrees=45
  swing1Motion=Free
  swing2LimitDegrees=45
  swing2Motion=Free
  twistLimitDegrees=45
  twistMotion=Free
 }
 {bone1Name=dyn_ft_l_02_tassle_01, bone2Name=dyn_ft_l_02_tassle_MSTR, swing1LimitDegrees=25, swing2LimitDegrees=5, twistLimitDegrees=1}
 {bone1Name=dyn_ft_l_02_tassle_02, bone2Name=dyn_ft_l_02_tassle_01, swing1LimitDegrees=25, swing2LimitDegrees=5, twistLimitDegrees=1}
 {bone1Name=dyn_ft_l_02_tassle_03, bone2Name=dyn_ft_l_02_tassle_02, swing1LimitDegrees=25, swing2LimitDegrees=5, twistLimitDegrees=1}
 {bone1Name=dyn_ft_l_02_tassle_04, bone2Name=dyn_ft_l_02_tassle_03, swing1LimitDegrees=25, swing2LimitDegrees=5, twistLimitDegrees=1}
 {
  bone1Name=dyn_ft_l_03_tassle_MSTR
  bone2Name=thigh_l
  swing1LimitDegrees=45
  swing1Motion=Free
  swing2LimitDegrees=45
  swing2Motion=Free
  twistLimitDegrees=45
  twistMotion=Free
 }
 {bone1Name=dyn_ft_l_03_tassle_01, bone2Name=dyn_ft_l_03_tassle_MSTR, swing1LimitDegrees=25, swing2LimitDegrees=5, twistLimitDegrees=1}
 {bone1Name=dyn_ft_l_03_tassle_02, bone2Name=dyn_ft_l_03_tassle_01, swing1LimitDegrees=25, swing2LimitDegrees=5, twistLimitDegrees=1}
 {bone1Name=dyn_ft_l_03_tassle_03, bone2Name=dyn_ft_l_03_tassle_02, swing1LimitDegrees=25, swing2LimitDegrees=5, twistLimitDegrees=1}
 {
  bone1Name=dyn_l_tassle_MSTR
  bone2Name=thigh_l
  swing1LimitDegrees=45
  swing1Motion=Free
  swing2LimitDegrees=45
  swing2Motion=Free
  twistLimitDegrees=45
  twistMotion=Free
 }
 {bone1Name=dyn_l_tassle_01, bone2Name=dyn_l_tassle_MSTR, swing1LimitDegrees=25, swing2LimitDegrees=5, twistLimitDegrees=1}
 {bone1Name=dyn_l_tassle_02, bone2Name=dyn_l_tassle_01, swing1LimitDegrees=25, swing2LimitDegrees=5, twistLimitDegrees=1}
 {bone1Name=dyn_l_tassle_03, bone2Name=dyn_l_tassle_02, swing1LimitDegrees=25, swing2LimitDegrees=5, twistLimitDegrees=1}
 {
  bone1Name=dyn_bk_r_01_tassle_MSTR
  bone2Name=thigh_r
  swing1LimitDegrees=45
  swing1Motion=Free
  swing2LimitDegrees=45
  swing2Motion=Free
  twistLimitDegrees=45
  twistMotion=Free
 }
 {bone1Name=dyn_bk_r_01_tassle_01, bone2Name=dyn_bk_r_01_tassle_MSTR, swing1LimitDegrees=25, swing2LimitDegrees=5, twistLimitDegrees=1}
 {
  bone1Name=dyn_bk_r_02_tassle_MSTR
  bone2Name=thigh_r
  swing1LimitDegrees=45
  swing1Motion=Free
  swing2LimitDegrees=45
  swing2Motion=Free
  twistLimitDegrees=45
  twistMotion=Free
 }
 {bone1Name=dyn_bk_r_02_tassle_01, bone2Name=dyn_bk_r_02_tassle_MSTR, swing1LimitDegrees=25, swing2LimitDegrees=5, twistLimitDegrees=1}
 {bone1Name=dyn_bk_r_02_tassle_02, bone2Name=dyn_bk_r_02_tassle_01, swing1LimitDegrees=25, swing2LimitDegrees=5, twistLimitDegrees=1}
 {
  bone1Name=dyn_ft_r_01_tassle_MSTR
  bone2Name=thigh_r
  swing1LimitDegrees=45
  swing1Motion=Free
  swing2LimitDegrees=45
  swing2Motion=Free
  twistLimitDegrees=45
  twistMotion=Free
 }
 {bone1Name=dyn_ft_r_01_tassle_01, bone2Name=dyn_ft_r_01_tassle_MSTR, swing1LimitDegrees=25, swing2LimitDegrees=5, twistLimitDegrees=1}
 {bone1Name=dyn_ft_r_01_tassle_02, bone2Name=dyn_ft_r_01_tassle_01, swing1LimitDegrees=25, swing2LimitDegrees=5, twistLimitDegrees=1}
 {bone1Name=dyn_ft_r_01_tassle_03, bone2Name=dyn_ft_r_01_tassle_02, swing1LimitDegrees=25, swing2LimitDegrees=5, twistLimitDegrees=1}
 {
  bone1Name=dyn_ft_r_02_tassle_MSTR
  bone2Name=thigh_r
  swing1LimitDegrees=45
  swing1Motion=Free
  swing2LimitDegrees=45
  swing2Motion=Free
  twistLimitDegrees=45
  twistMotion=Free
 }
 {bone1Name=dyn_ft_r_02_tassle_01, bone2Name=dyn_ft_r_02_tassle_MSTR, swing1LimitDegrees=25, swing2LimitDegrees=5, twistLimitDegrees=1}
 {bone1Name=dyn_ft_r_02_tassle_02, bone2Name=dyn_ft_r_02_tassle_01, swing1LimitDegrees=25, swing2LimitDegrees=5, twistLimitDegrees=1}
 {bone1Name=dyn_ft_r_02_tassle_03, bone2Name=dyn_ft_r_02_tassle_02, swing1LimitDegrees=25, swing2LimitDegrees=5, twistLimitDegrees=1}
 {
  bone1Name=dyn_ft_r_03_tassle_MSTR
  bone2Name=thigh_r
  swing1LimitDegrees=45
  swing1Motion=Free
  swing2LimitDegrees=45
  swing2Motion=Free
  twistLimitDegrees=45
  twistMotion=Free
 }
 {bone1Name=dyn_ft_r_03_tassle_01, bone2Name=dyn_ft_r_03_tassle_MSTR, swing1LimitDegrees=25, swing2LimitDegrees=5, twistLimitDegrees=1}
 {bone1Name=dyn_ft_r_03_tassle_02, bone2Name=dyn_ft_r_03_tassle_01, swing1LimitDegrees=25, swing2LimitDegrees=5, twistLimitDegrees=1}
 {bone1Name=dyn_ft_r_03_tassle_03, bone2Name=dyn_ft_r_03_tassle_02, swing1LimitDegrees=25, swing2LimitDegrees=5, twistLimitDegrees=1}
 {
  bone1Name=dyn_r_tassle_MSTR
  bone2Name=thigh_r
  swing1LimitDegrees=45
  swing1Motion=Free
  swing2LimitDegrees=45
  swing2Motion=Free
  twistLimitDegrees=45
  twistMotion=Free
 }
 {bone1Name=dyn_r_tassle_01, bone2Name=dyn_r_tassle_MSTR, swing1LimitDegrees=25, swing2LimitDegrees=5, twistLimitDegrees=1}
 {bone1Name=dyn_r_tassle_02, bone2Name=dyn_r_tassle_01, swing1LimitDegrees=25, swing2LimitDegrees=5, twistLimitDegrees=1}
 {bone1Name=dyn_r_tassle_03, bone2Name=dyn_r_tassle_02, swing1LimitDegrees=25, swing2LimitDegrees=5, twistLimitDegrees=1}
 {bone1Name=dyn_belt_03, bone2Name=dyn_belt_02, swing1LimitDegrees=15, swing2LimitDegrees=25, twistLimitDegrees=1}
 {
  bone1Name=dyn_belt_MSTR
  bone2Name=spine_02
  swing1LimitDegrees=45
  swing1Motion=Free
  swing2LimitDegrees=45
  swing2Motion=Free
  twistLimitDegrees=45
  twistMotion=Free
 }
 {bone1Name=dyn_belt_01, bone2Name=dyn_belt_MSTR, swing1LimitDegrees=15, swing2LimitDegrees=25, twistLimitDegrees=1}
 {bone1Name=dyn_belt_02, bone2Name=dyn_belt_01, swing1LimitDegrees=15, swing2LimitDegrees=25, twistLimitDegrees=1}
]
