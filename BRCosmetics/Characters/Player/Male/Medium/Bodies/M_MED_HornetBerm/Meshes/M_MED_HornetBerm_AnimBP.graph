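# Post-process AnimBP for the HornetBerm male body cosmetic, adding secondary motion to its strings and cape.
# The copied parent pose runs through Control Rig, RigidBody, constraints and AnimDynamics into a shared cache.
# Switches then select the cape solution: cloth, RigidBody with a backbling variant, AnimDynamics, or trail controllers with a DBNO variant.
# The EventGraph updates weapon, emote and backbling state, drives cloth settings and toggles a pickaxe-dependent material element.
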AnimBlueprint /BRCosmetics/Characters/Player/Male/Medium/Bodies/M_MED_HornetBerm/Meshes/M_MED_HornetBerm_AnimBP
target_skeleton=&/M_MED_HornetBerm_Skeleton
graphs [
 AnimGraph {nodes [
   n1 = Blend Poses by bool(BlendPose_0=Use cached pose 'dbnoCache'(), BlendPose_1=Use cached pose 'trailCache'(), bActiveValue=Get bIsDBNO()) {blendType=Linear}
   n2 = Use cached pose 'animDyn Cache'()
   n3 = Get bAnimDynamicsEnabled() -> (bAnimDynamicsEnabled:bool)
   Output Pose(Result=Blend Poses by bool(
     BlendPose_0=Use cached pose 'RbanCache'()
     BlendPose_1=Blend Poses by bool(BlendPose_0=@n2, BlendPose_1=@n1, bActiveValue=@n3) {blendType=Linear}
     bActiveValue=Get bRigidBodyEnabled()
    ) {blendType=Linear}) {layerGroup=None}
   n4 = Control Rig!(
    Source=Copy Pose From Mesh() {bUseAttachedParent=true}
    Alpha
    Is not playing emote Alpha:double=Get Is not Playing Emote Alpha() -> (Is not Playing Emote Alpha:double)
    IsPistol:double=Get Is Pistol Alpha() -> (Is Pistol Alpha:double)
   ) -> (Pose) {controlRigAssetReference {controlRigAsset=&/M_MED_HornetBerm_CtrlRig}}
   n5 = RigidBody(ComponentPose=Local To Component(LocalPose=@n4)) {
    bClampLinearTranslationLimitToRefPose=true
    bOverrideWorldGravity=true
    overridePhysicsAsset=&/M_MED_HornetBerm_Physics
    simulationSpace=BaseBoneSpace
    alphaBoolBlend {blendOption=Linear}
    baseBoneRef {boneName=pelvis}
    overrideWorldGravity {z=-980}
    simSpaceSettings {
     maxAngularAcceleration=40
     maxAngularVelocity=8
     maxLinearAcceleration=3500
     maxLinearVelocity=700
     worldAlpha=0.5
     externalLinearDragV=<3 3 3>
    }
   }
   n6 = "Constraint - dyn_l_sidestring_mstr (PPPP)"!(
    ComponentPose=@n5
    Alpha=1
    ConstraintWeights_0=0.1
    ConstraintWeights_1:float=0.1
    ConstraintWeights_2:float=0.4
    ConstraintWeights_3:float=0.4
   ) -> (Pose) {
    boneToModify {boneName=dyn_l_sidestring_mstr}
    constraintSetup [{targetBone {boneName=spine_04}}, {targetBone {boneName=clavicle_l}}, {targetBone {boneName=spine_03}}, {targetBone {boneName=spine_02}}]
    constraintWeights [1,1,1,1]
   }
   n7 = "Constraint - dyn_r_sidestring_mstr (PPPP)"!(
    ComponentPose=@n6
    Alpha=1
    ConstraintWeights_0=0.1
    ConstraintWeights_1:float=0.1
    ConstraintWeights_2:float=0.4
    ConstraintWeights_3:float=0.4
   ) -> (Pose) {
    boneToModify {boneName=dyn_r_sidestring_mstr}
    constraintSetup [{targetBone {boneName=spine_04}}, {targetBone {boneName=clavicle_r}}, {targetBone {boneName=spine_03}}, {targetBone {boneName=spine_02}}]
    constraintWeights [1,1,1,1]
   }
   n8 = "Transform (Modify) Bone - Bone: backpack_collision"(ComponentPose=@n7, Alpha=1, Translation="0,0,200", Rotation="0,0,0", Scale="1,1,1") {
    rotationMode=BMM_Additive
    rotationSpace=BCS_BoneSpace
    translationMode=BMM_Additive
    translationSpace=BCS_BoneSpace
    boneToModify {boneName=backpack_collision}
    translation {z=200}
   }
   n9 = "Transform (Modify) Bone - Bone: backpack_collision"(ComponentPose=@n8, Alpha=1, Translation="0,-200,200", Rotation="0,0,0", Scale="1,1,1") {
    rotationMode=BMM_Additive
    rotationSpace=BCS_BoneSpace
    translationMode=BMM_Additive
    translationSpace=BCS_BoneSpace
    boneToModify {boneName=backpack_collision}
    translation {y=-200, z=200}
   }
   n10 = Anim Dynamics - Bone: dyn_l_bracerstring_01(ComponentPose=@n9, Alpha=1, ExternalForce="0,0,0") {
    angularBiasOverride=1
    angularDampingOverride=1
    angularSpringConstant=500
    bAngularSpring=true
    bOverrideAngularBias=true
    bOverrideAngularDamping=true
    bUseGravityOverride=true
    gravityScale=0.35
    simulationSpace=RootRelative
    alphaBoolBlend {blendOption=Linear}
    boundBone {boneName=dyn_l_bracerstring_01}
    gravityOverride {z=-980}
    physicsBodyDefinitions [{
      boundBone {boneName=dyn_l_bracerstring_01}
      boxExtents=<14 14 14>
      constraintSetup {angularLimitsMax=<1 45 15>, angularLimitsMin=<-1 0 -15>, angularTarget=<1 0 0>}
      localJointOffset=<2 0 0>
     }]
    simSpaceSettings {maxAngularAcceleration=45, maxAngularVelocity=8, simSpaceAngularAlpha=0.25}
   }
   n11 = Anim Dynamics - Bone: dyn_l_string_01(ComponentPose=@n10, Alpha=1, ExternalForce="0,0,0") {
    angularBiasOverride=1
    angularDampingOverride=1
    angularSpringConstant=500
    bAngularSpring=true
    bOverrideAngularBias=true
    bOverrideAngularDamping=true
    bUseGravityOverride=true
    gravityScale=0.35
    simulationSpace=RootRelative
    alphaBoolBlend {blendOption=Linear}
    boundBone {boneName=dyn_l_string_01}
    gravityOverride {z=-980}
    physicsBodyDefinitions [{
      boundBone {boneName=dyn_l_string_01}
      boxExtents=<14 14 14>
      constraintSetup {angularLimitsMax=<1 45 15>, angularLimitsMin=<-1 0 -15>, angularTarget=<1 0 0>}
      localJointOffset=<2 0 0>
     }]
    simSpaceSettings {maxAngularAcceleration=45, maxAngularVelocity=8, simSpaceAngularAlpha=0.25}
   }
   n12 = Anim Dynamics - Bone: dyn_l_string_02(ComponentPose=@n11, Alpha=1, ExternalForce="0,0,0") {
    angularBiasOverride=1
    angularDampingOverride=1
    angularSpringConstant=500
    bAngularSpring=true
    bOverrideAngularBias=true
    bOverrideAngularDamping=true
    bUseGravityOverride=true
    gravityScale=0.35
    simulationSpace=RootRelative
    alphaBoolBlend {blendOption=Linear}
    boundBone {boneName=dyn_l_string_02}
    gravityOverride {z=-980}
    physicsBodyDefinitions [{
      boundBone {boneName=dyn_l_string_02}
      boxExtents=<14 14 14>
      constraintSetup {angularLimitsMax=<1 45 15>, angularLimitsMin=<-1 0 -15>, angularTarget=<1 0 0>}
      localJointOffset=<2 0 0>
     }]
    simSpaceSettings {maxAngularAcceleration=45, maxAngularVelocity=8, simSpaceAngularAlpha=0.25}
   }
   n13 = Anim Dynamics - Bone: dyn_r_bracerstring_01(ComponentPose=@n12, Alpha=1, ExternalForce="0,0,0") {
    angularBiasOverride=1
    angularDampingOverride=1
    angularSpringConstant=500
    bAngularSpring=true
    bOverrideAngularBias=true
    bOverrideAngularDamping=true
    bUseGravityOverride=true
    gravityScale=0.35
    simulationSpace=RootRelative
    alphaBoolBlend {blendOption=Linear}
    boundBone {boneName=dyn_r_bracerstring_01}
    gravityOverride {z=-980}
    physicsBodyDefinitions [{
      boundBone {boneName=dyn_r_bracerstring_01}
      boxExtents=<14 14 14>
      constraintSetup {angularLimitsMax=<1 45 15>, angularLimitsMin=<-1 0 -15>, angularTarget=<1 0 0>}
      localJointOffset=<2 0 0>
     }]
    simSpaceSettings {maxAngularAcceleration=45, maxAngularVelocity=8, simSpaceAngularAlpha=0.25}
   }
   n14 = Anim Dynamics - Bone: dyn_r_string_01(ComponentPose=@n13, Alpha=1, ExternalForce="0,0,0") {
    angularBiasOverride=1
    angularDampingOverride=1
    angularSpringConstant=500
    bAngularSpring=true
    bOverrideAngularBias=true
    bOverrideAngularDamping=true
    bUseGravityOverride=true
    gravityScale=0.35
    simulationSpace=RootRelative
    alphaBoolBlend {blendOption=Linear}
    boundBone {boneName=dyn_r_string_01}
    gravityOverride {z=-980}
    physicsBodyDefinitions [{
      boundBone {boneName=dyn_r_string_01}
      boxExtents=<14 14 14>
      constraintSetup {angularLimitsMax=<1 45 15>, angularLimitsMin=<-1 0 -15>, angularTarget=<1 0 0>}
      localJointOffset=<2 0 0>
     }]
    simSpaceSettings {maxAngularAcceleration=45, maxAngularVelocity=8, simSpaceAngularAlpha=0.25}
   }
   n15 = Anim Dynamics - Bone: dyn_r_string_02(ComponentPose=@n14, Alpha=1, ExternalForce="0,0,0") {
    angularBiasOverride=1
    angularDampingOverride=1
    angularSpringConstant=500
    bAngularSpring=true
    bOverrideAngularBias=true
    bOverrideAngularDamping=true
    bUseGravityOverride=true
    gravityScale=0.35
    simulationSpace=RootRelative
    alphaBoolBlend {blendOption=Linear}
    boundBone {boneName=dyn_r_string_02}
    gravityOverride {z=-980}
    physicsBodyDefinitions [{
      boundBone {boneName=dyn_r_string_02}
      boxExtents=<14 14 14>
      constraintSetup {angularLimitsMax=<1 45 15>, angularLimitsMin=<-1 0 -15>, angularTarget=<1 0 0>}
      localJointOffset=<2 0 0>
     }]
    simSpaceSettings {maxAngularAcceleration=45, maxAngularVelocity=8, simSpaceAngularAlpha=0.25}
   }
   n16 = Anim Dynamics - Bone: dyn_l_sidestring_01(ComponentPose=@n15, Alpha=1, ExternalForce="0,0,0") {
    angularBiasOverride=1
    angularDampingOverride=1
    angularSpringConstant=500
    bAngularSpring=true
    bOverrideAngularBias=true
    bOverrideAngularDamping=true
    bUseGravityOverride=true
    gravityScale=0.35
    simulationSpace=RootRelative
    alphaBoolBlend {blendOption=Linear}
    boundBone {boneName=dyn_l_sidestring_01}
    componentLinearAccScale=<0.3 0.3 0.4>
    componentLinearVelScale=<0.3 0.3 0.4>
    gravityOverride {z=-980}
    physicsBodyDefinitions [{
      boundBone {boneName=dyn_l_sidestring_01}
      boxExtents=<14 14 14>
      constraintSetup {angularLimitsMax=<1 0 5>, angularLimitsMin=<-1 -45 -5>, angularTarget=<1 0 0>}
      localJointOffset=<4 0 0>
     }]
    simSpaceSettings {maxAngularAcceleration=45, maxAngularVelocity=8, simSpaceAngularAlpha=0.25}
   }
   n17 = Anim Dynamics - Bone: dyn_l_cheststring_01(ComponentPose=@n16, Alpha=1, ExternalForce="0,0,0") {
    angularBiasOverride=1
    angularDampingOverride=1
    angularSpringConstant=500
    bAngularSpring=true
    bOverrideAngularBias=true
    bOverrideAngularDamping=true
    bUseGravityOverride=true
    gravityScale=0.35
    simulationSpace=RootRelative
    alphaBoolBlend {blendOption=Linear}
    boundBone {boneName=dyn_l_cheststring_01}
    componentLinearAccScale=<0.3 0.3 0.4>
    componentLinearVelScale=<0.3 0.3 0.4>
    gravityOverride {z=-980}
    physicsBodyDefinitions [{
      boundBone {boneName=dyn_l_cheststring_01}
      boxExtents=<14 14 14>
      constraintSetup {angularLimitsMax=<1 0 5>, angularLimitsMin=<-1 -45 -5>, angularTarget=<1 0 0>}
      localJointOffset=<4 0 0>
     }]
    simSpaceSettings {maxAngularAcceleration=45, maxAngularVelocity=8, simSpaceAngularAlpha=0.25}
   }
   n18 = Anim Dynamics - Bone: dyn_l_cheststring_02(ComponentPose=@n17, Alpha=1, ExternalForce="0,0,0") {
    angularBiasOverride=1
    angularDampingOverride=1
    angularSpringConstant=500
    bAngularSpring=true
    bOverrideAngularBias=true
    bOverrideAngularDamping=true
    bUseGravityOverride=true
    gravityScale=0.35
    simulationSpace=RootRelative
    alphaBoolBlend {blendOption=Linear}
    boundBone {boneName=dyn_l_cheststring_02}
    componentLinearAccScale=<0.3 0.3 0.4>
    componentLinearVelScale=<0.3 0.3 0.4>
    gravityOverride {z=-980}
    physicsBodyDefinitions [{
      boundBone {boneName=dyn_l_cheststring_02}
      boxExtents=<14 14 14>
      constraintSetup {angularLimitsMax=<1 0 5>, angularLimitsMin=<-1 -45 -5>, angularTarget=<1 0 0>}
      localJointOffset=<4 0 0>
     }]
    simSpaceSettings {maxAngularAcceleration=45, maxAngularVelocity=8, simSpaceAngularAlpha=0.25}
   }
   n19 = Anim Dynamics - Bone: dyn_r_sidestring_01(ComponentPose=@n18, Alpha=1, ExternalForce="0,0,0") {
    angularBiasOverride=1
    angularDampingOverride=1
    angularSpringConstant=500
    bAngularSpring=true
    bOverrideAngularBias=true
    bOverrideAngularDamping=true
    bUseGravityOverride=true
    gravityScale=0.35
    simulationSpace=RootRelative
    alphaBoolBlend {blendOption=Linear}
    boundBone {boneName=dyn_r_sidestring_01}
    componentLinearAccScale=<0.3 0.3 0.4>
    componentLinearVelScale=<0.3 0.3 0.4>
    gravityOverride {z=-980}
    physicsBodyDefinitions [{
      boundBone {boneName=dyn_r_sidestring_01}
      boxExtents=<14 14 14>
      constraintSetup {angularLimitsMax=<1 0 5>, angularLimitsMin=<-1 -45 -5>, angularTarget=<1 0 0>}
      localJointOffset=<4 0 0>
     }]
    simSpaceSettings {maxAngularAcceleration=45, maxAngularVelocity=8, simSpaceAngularAlpha=0.25}
   }
   n20 = Anim Dynamics - Bone: dyn_r_cheststring_01(ComponentPose=@n19, Alpha=1, ExternalForce="0,0,0") {
    angularBiasOverride=1
    angularDampingOverride=1
    angularSpringConstant=500
    bAngularSpring=true
    bOverrideAngularBias=true
    bOverrideAngularDamping=true
    bUseGravityOverride=true
    gravityScale=0.35
    simulationSpace=RootRelative
    alphaBoolBlend {blendOption=Linear}
    boundBone {boneName=dyn_r_cheststring_01}
    componentLinearAccScale=<0.3 0.3 0.4>
    componentLinearVelScale=<0.3 0.3 0.4>
    gravityOverride {z=-980}
    physicsBodyDefinitions [{
      boundBone {boneName=dyn_r_cheststring_01}
      boxExtents=<14 14 14>
      constraintSetup {angularLimitsMax=<1 0 5>, angularLimitsMin=<-1 -45 -5>, angularTarget=<1 0 0>}
      localJointOffset=<4 0 0>
     }]
    simSpaceSettings {maxAngularAcceleration=45, maxAngularVelocity=8, simSpaceAngularAlpha=0.25}
   }
   n21 = Anim Dynamics - Bone: dyn_r_cheststring_02(ComponentPose=@n20, Alpha=1, ExternalForce="0,0,0") {
    angularBiasOverride=1
    angularDampingOverride=1
    angularSpringConstant=500
    bAngularSpring=true
    bOverrideAngularBias=true
    bOverrideAngularDamping=true
    bUseGravityOverride=true
    gravityScale=0.35
    simulationSpace=RootRelative
    alphaBoolBlend {blendOption=Linear}
    boundBone {boneName=dyn_r_cheststring_02}
    componentLinearAccScale=<0.3 0.3 0.4>
    componentLinearVelScale=<0.3 0.3 0.4>
    gravityOverride {z=-980}
    physicsBodyDefinitions [{
      boundBone {boneName=dyn_r_cheststring_02}
      boxExtents=<14 14 14>
      constraintSetup {angularLimitsMax=<1 0 5>, angularLimitsMin=<-1 -45 -5>, angularTarget=<1 0 0>}
      localJointOffset=<4 0 0>
     }]
    simSpaceSettings {maxAngularAcceleration=45, maxAngularVelocity=8, simSpaceAngularAlpha=0.25}
   }
   Save cached pose 'Start_Cache'(Pose=Component To Local(ComponentPose=@n21)) {cacheName=Start_Cache}
   n22 = "Transform (Modify) Bone - Bone: dyn_cape_l_01"(ComponentPose=Local To Component(LocalPose=Use cached pose 'Start_Cache'()), Alpha=1, Translation="0,0,0", Rotation="-10,0,0", Scale="1,1,1") {
    rotationMode=BMM_Additive
    rotationSpace=BCS_BoneSpace
    alphaBoolBlend {blendOption=Linear}
    boneToModify {boneName=dyn_cape_l_01}
    rotation {pitch=-10}
   }
   n23 = "Transform (Modify) Bone - Bone: dyn_cape_01"(ComponentPose=@n22, Alpha=1, Translation="0,0,0", Rotation="-7,0,0", Scale="1,1,1") {
    rotationMode=BMM_Additive
    rotationSpace=BCS_BoneSpace
    alphaBoolBlend {blendOption=Linear}
    boneToModify {boneName=dyn_cape_01}
    rotation {pitch=-7}
   }
   n24 = "Transform (Modify) Bone - Bone: dyn_cape_r_01"(ComponentPose=@n23, Alpha=1, Translation="0,0,0", Rotation="0,0,0", Scale="1,1,1") {rotationMode=BMM_Additive, rotationSpace=BCS_BoneSpace, alphaBoolBlend {blendOption=Linear}, boneToModify {boneName=dyn_cape_r_01}}
   n25 = "Transform (Modify) Bone - Bone: dyn_cape_l_03"(ComponentPose=@n24, Alpha=1, Translation="0,0,0", Rotation="-10,0,0", Scale="1,1,1") {
    rotationMode=BMM_Additive
    rotationSpace=BCS_BoneSpace
    alphaBoolBlend {blendOption=Linear}
    boneToModify {boneName=dyn_cape_l_03}
    rotation {pitch=-10}
   }
   n26 = "Transform (Modify) Bone - Bone: dyn_cape_03"(ComponentPose=@n25, Alpha=1, Translation="0,0,0", Rotation="-7,0,0", Scale="1,1,1") {
    rotationMode=BMM_Additive
    rotationSpace=BCS_BoneSpace
    alphaBoolBlend {blendOption=Linear}
    boneToModify {boneName=dyn_cape_03}
    rotation {pitch=-7}
   }
   n27 = "Transform (Modify) Bone - Bone: dyn_cape_r_03"(ComponentPose=@n26, Alpha=1, Translation="0,0,0", Rotation="-3,0,0", Scale="1,1,1") {
    rotationMode=BMM_Additive
    rotationSpace=BCS_BoneSpace
    alphaBoolBlend {blendOption=Linear}
    boneToModify {boneName=dyn_cape_r_03}
    rotation {pitch=-3}
   }
   n28 = "Transform (Modify) Bone - Bone: dyn_cape_l_04"(ComponentPose=@n27, Alpha=1, Translation="0,0,0", Rotation="-10,0,0", Scale="1,1,1") {
    rotationMode=BMM_Additive
    rotationSpace=BCS_BoneSpace
    alphaBoolBlend {blendOption=Linear}
    boneToModify {boneName=dyn_cape_l_04}
    rotation {pitch=-10}
   }
   n29 = "Transform (Modify) Bone - Bone: dyn_cape_04"(ComponentPose=@n28, Alpha=1, Translation="0,0,0", Rotation="-10,0,0", Scale="1,1,1") {
    rotationMode=BMM_Additive
    rotationSpace=BCS_BoneSpace
    alphaBoolBlend {blendOption=Linear}
    boneToModify {boneName=dyn_cape_04}
    rotation {pitch=-10}
   }
   n30 = "Transform (Modify) Bone - Bone: dyn_cape_r_04"(ComponentPose=@n29, Alpha=1, Translation="0,0,0", Rotation="-20,0,0", Scale="1,1,1") {
    rotationMode=BMM_Additive
    rotationSpace=BCS_BoneSpace
    alphaBoolBlend {blendOption=Linear}
    boneToModify {boneName=dyn_cape_r_04}
    rotation {pitch=-20}
   }
   n31 = Get BwdMovement() -> (BwdMovement:double)
   n32 = Trail controller - Bone: dyn_cape_l_07(ComponentPose=@n30, Alpha=@n31) {
    bActorSpaceFakeVel=true
    bShowTrailLocation=true
    bUsePlanarLimit=true
    chainLength=7
    debugLifeTime=0.1
    maxDeltaTime=0.0333
    relaxationSpeedScale=2
    alphaScaleBiasClamp {bInterpResult=true, interpSpeedDecreasing=25, interpSpeedIncreasing=25}
    baseJoint {boneName=backpack_root}
    rotationLimits [
     {limitMax=<180 180 180>}
     {limitMax=<180 180 180>}
     {limitMax=<180 180 180>}
     {limitMax=<180 180 180>}
     {limitMax=<180 180 180>}
     {limitMax=<180 180 180>}
     {limitMax=<180 180 180>}
    ]
    rotationOffsets [<0 0 0>, <0 0 0>, <0 0 0>, <0 0 0>, <0 0 0>, <0 0 0>, <0 0 0>]
    trailBone {boneName=dyn_cape_l_07}
    trailRelaxationSpeed {editorCurveData {keys [<at 0 23 RCIM_Linear>,<at 0.284 19.3799 RCIM_Linear>,<at 0.6554 8.5924 RCIM_Linear>,<at 1 5 RCIM_Linear>]}}
   }
   n33 = Trail controller - Bone: dyn_cape_07(ComponentPose=@n32, Alpha=@n31) {
    bActorSpaceFakeVel=true
    bShowTrailLocation=true
    bUsePlanarLimit=true
    chainLength=7
    debugLifeTime=0.1
    maxDeltaTime=0.0333
    relaxationSpeedScale=2
    alphaScaleBiasClamp {bInterpResult=true, interpSpeedDecreasing=25, interpSpeedIncreasing=25}
    baseJoint {boneName=backpack_root}
    rotationLimits [
     {limitMax=<180 180 180>}
     {limitMax=<180 180 180>}
     {limitMax=<180 180 180>}
     {limitMax=<180 180 180>}
     {limitMax=<180 180 180>}
     {limitMax=<180 180 180>}
     {limitMax=<180 180 180>}
    ]
    rotationOffsets [<0 0 0>, <0 0 0>, <0 0 0>, <0 0 0>, <0 0 0>, <0 0 0>, <0 0 0>]
    trailBone {boneName=dyn_cape_07}
    trailRelaxationSpeed {editorCurveData {keys [<at 0 23 RCIM_Linear>,<at 0.284 19.3799 RCIM_Linear>,<at 0.6554 8.5924 RCIM_Linear>,<at 1 5 RCIM_Linear>]}}
   }
   n34 = Trail controller - Bone: dyn_cape_r_07(ComponentPose=@n33, Alpha=@n31) {
    bActorSpaceFakeVel=true
    bShowTrailLocation=true
    bUsePlanarLimit=true
    chainLength=7
    debugLifeTime=0.1
    maxDeltaTime=0.0333
    relaxationSpeedScale=2
    alphaScaleBiasClamp {bInterpResult=true, interpSpeedDecreasing=25, interpSpeedIncreasing=25}
    baseJoint {boneName=backpack_root}
    rotationLimits [
     {limitMax=<180 180 180>}
     {limitMax=<180 180 180>}
     {limitMax=<180 180 180>}
     {limitMax=<180 180 180>}
     {limitMax=<180 180 180>}
     {limitMax=<180 180 180>}
     {limitMax=<180 180 180>}
    ]
    rotationOffsets [<0 0 0>, <0 0 0>, <0 0 0>, <0 0 0>, <0 0 0>, <0 0 0>, <0 0 0>]
    trailBone {boneName=dyn_cape_r_07}
    trailRelaxationSpeed {editorCurveData {keys [<at 0 23 RCIM_Linear>,<at 0.284 19.3799 RCIM_Linear>,<at 0.6554 8.5924 RCIM_Linear>,<at 1 5 RCIM_Linear>]}}
   }
   Save cached pose 'dbnoCache'(Pose=Component To Local(ComponentPose=@n34)) {cacheName=dbnoCache}
   n35 = Anim Dynamics - Bone: dyn_cape_01(ComponentPose=Local To Component(LocalPose=Use cached pose 'Start_Cache'()), Alpha=1, ExternalForce="0,0,0") {
    angularDampingOverride=1
    angularSpringConstant=10000
    bAngularSpring=true
    bOverrideAngularDamping=true
    bUseGravityOverride=true
    gravityScale=0.35
    alphaBoolBlend {blendOption=Linear}
    boundBone {boneName=dyn_cape_01}
    componentAppliedLinearAccClamp=<3000 3000 3000>
    componentLinearAccScale=<2 2 2>
    componentLinearVelScale=<2 2 2>
    gravityOverride {z=-980}
    physicsBodyDefinitions [{
      boundBone {boneName=dyn_cape_01}
      boxExtents=<10 40 40>
      constraintSetup {angularLimitsMax=<0 5 15>, angularLimitsMin=<0 -40 -15>, angularTarget=<1 0 0>}
      localJointOffset=<5 0 0>
     }]
    relativeSpaceBone {boneName=backpack_root}
   }
   n36 = Anim Dynamics - Bone: dyn_cape_03(ComponentPose=@n35, Alpha=1, ExternalForce="0,0,0") {
    angularDampingOverride=1
    angularSpringConstant=10000
    bAngularSpring=true
    bOverrideAngularDamping=true
    bUseGravityOverride=true
    gravityScale=0.35
    alphaBoolBlend {blendOption=Linear}
    boundBone {boneName=dyn_cape_03}
    componentAppliedLinearAccClamp=<3000 3000 3000>
    componentLinearAccScale=<2 2 2>
    componentLinearVelScale=<2 2 2>
    gravityOverride {z=-980}
    physicsBodyDefinitions [{
      boundBone {boneName=dyn_cape_03}
      boxExtents=<10 40 40>
      constraintSetup {angularLimitsMax=<0 5 15>, angularLimitsMin=<0 -40 -15>, angularTarget=<1 0 0>}
      localJointOffset=<5 0 0>
     }]
    relativeSpaceBone {boneName=backpack_root}
   }
   n37 = Anim Dynamics - Bone: dyn_cape_05(ComponentPose=@n36, Alpha=1, ExternalForce="0,0,0") {
    angularDampingOverride=1
    angularSpringConstant=10000
    bAngularSpring=true
    bOverrideAngularDamping=true
    bUseGravityOverride=true
    gravityScale=0.35
    alphaBoolBlend {blendOption=Linear}
    boundBone {boneName=dyn_cape_05}
    componentAppliedLinearAccClamp=<3000 3000 3000>
    componentLinearAccScale=<2 2 2>
    componentLinearVelScale=<2 2 2>
    gravityOverride {z=-980}
    physicsBodyDefinitions [{
      boundBone {boneName=dyn_cape_05}
      boxExtents=<10 40 40>
      constraintSetup {angularLimitsMax=<0 5 15>, angularLimitsMin=<0 -40 -15>, angularTarget=<1 0 0>}
      localJointOffset=<5 0 0>
     }]
    relativeSpaceBone {boneName=backpack_root}
   }
   n38 = Anim Dynamics - Bone: dyn_cape_06(ComponentPose=@n37, Alpha=1, ExternalForce="0,0,0") {
    angularDampingOverride=1
    angularSpringConstant=10000
    bAngularSpring=true
    bOverrideAngularDamping=true
    bUseGravityOverride=true
    gravityScale=0.35
    alphaBoolBlend {blendOption=Linear}
    boundBone {boneName=dyn_cape_06}
    componentAppliedLinearAccClamp=<3000 3000 3000>
    componentLinearAccScale=<2 2 2>
    componentLinearVelScale=<2 2 2>
    gravityOverride {z=-980}
    physicsBodyDefinitions [{
      boundBone {boneName=dyn_cape_06}
      boxExtents=<10 40 40>
      constraintSetup {angularLimitsMax=<0 5 15>, angularLimitsMin=<0 -40 -15>, angularTarget=<1 0 0>}
      localJointOffset=<5 0 0>
     }]
    relativeSpaceBone {boneName=backpack_root}
   }
   n39 = Anim Dynamics - Bone: dyn_cape_r_01(ComponentPose=@n38, Alpha=1, ExternalForce="0,0,0") {
    angularDampingOverride=1
    angularSpringConstant=10000
    bAngularSpring=true
    bOverrideAngularDamping=true
    bUseGravityOverride=true
    gravityScale=0.35
    alphaBoolBlend {blendOption=Linear}
    boundBone {boneName=dyn_cape_r_01}
    componentAppliedLinearAccClamp=<3000 3000 3000>
    componentLinearAccScale=<2 2 2>
    componentLinearVelScale=<2 2 2>
    gravityOverride {z=-980}
    physicsBodyDefinitions [{
      boundBone {boneName=dyn_cape_r_01}
      boxExtents=<10 40 40>
      constraintSetup {angularLimitsMax=<0 5 15>, angularLimitsMin=<0 -40 -15>, angularTarget=<1 0 0>}
      localJointOffset=<5 0 0>
     }]
    relativeSpaceBone {boneName=backpack_root}
   }
   n40 = Anim Dynamics - Bone: dyn_cape_r_03(ComponentPose=@n39, Alpha=1, ExternalForce="0,0,0") {
    angularDampingOverride=1
    angularSpringConstant=10000
    bAngularSpring=true
    bOverrideAngularDamping=true
    bUseGravityOverride=true
    gravityScale=0.35
    alphaBoolBlend {blendOption=Linear}
    boundBone {boneName=dyn_cape_r_03}
    componentAppliedLinearAccClamp=<3000 3000 3000>
    componentLinearAccScale=<2 2 2>
    componentLinearVelScale=<2 2 2>
    gravityOverride {z=-980}
    physicsBodyDefinitions [{
      boundBone {boneName=dyn_cape_r_03}
      boxExtents=<10 40 40>
      constraintSetup {angularLimitsMax=<0 5 15>, angularLimitsMin=<0 -40 -15>, angularTarget=<1 0 0>}
      localJointOffset=<5 0 0>
     }]
    relativeSpaceBone {boneName=backpack_root}
   }
   n41 = Anim Dynamics - Bone: dyn_cape_r_05(ComponentPose=@n40, Alpha=1, ExternalForce="0,0,0") {
    angularDampingOverride=1
    angularSpringConstant=10000
    bAngularSpring=true
    bOverrideAngularDamping=true
    bUseGravityOverride=true
    gravityScale=0.35
    alphaBoolBlend {blendOption=Linear}
    boundBone {boneName=dyn_cape_r_05}
    componentAppliedLinearAccClamp=<3000 3000 3000>
    componentLinearAccScale=<2 2 2>
    componentLinearVelScale=<2 2 2>
    gravityOverride {z=-980}
    physicsBodyDefinitions [{
      boundBone {boneName=dyn_cape_r_05}
      boxExtents=<10 40 40>
      constraintSetup {angularLimitsMax=<0 5 15>, angularLimitsMin=<0 -40 -15>, angularTarget=<1 0 0>}
      localJointOffset=<5 0 0>
     }]
    relativeSpaceBone {boneName=backpack_root}
   }
   n42 = Anim Dynamics - Bone: dyn_cape_r_06(ComponentPose=@n41, Alpha=1, ExternalForce="0,0,0") {
    angularDampingOverride=1
    angularSpringConstant=10000
    bAngularSpring=true
    bOverrideAngularDamping=true
    bUseGravityOverride=true
    gravityScale=0.35
    alphaBoolBlend {blendOption=Linear}
    boundBone {boneName=dyn_cape_r_06}
    componentAppliedLinearAccClamp=<3000 3000 3000>
    componentLinearAccScale=<2 2 2>
    componentLinearVelScale=<2 2 2>
    gravityOverride {z=-980}
    physicsBodyDefinitions [{
      boundBone {boneName=dyn_cape_r_06}
      boxExtents=<10 40 40>
      constraintSetup {angularLimitsMax=<0 5 15>, angularLimitsMin=<0 -40 -15>, angularTarget=<1 0 0>}
      localJointOffset=<5 0 0>
     }]
    relativeSpaceBone {boneName=backpack_root}
   }
   n43 = Anim Dynamics - Bone: dyn_cape_l_01(ComponentPose=@n42, Alpha=1, ExternalForce="0,0,0") {
    angularDampingOverride=1
    angularSpringConstant=10000
    bAngularSpring=true
    bOverrideAngularDamping=true
    bUseGravityOverride=true
    gravityScale=0.35
    alphaBoolBlend {blendOption=Linear}
    boundBone {boneName=dyn_cape_l_01}
    componentAppliedLinearAccClamp=<3000 3000 3000>
    componentLinearAccScale=<2 2 2>
    componentLinearVelScale=<2 2 2>
    gravityOverride {z=-980}
    physicsBodyDefinitions [{
      boundBone {boneName=dyn_cape_l_01}
      boxExtents=<10 40 40>
      constraintSetup {angularLimitsMax=<0 5 15>, angularLimitsMin=<0 -40 -15>, angularTarget=<1 0 0>}
      localJointOffset=<5 0 0>
     }]
    relativeSpaceBone {boneName=backpack_root}
   }
   n44 = Anim Dynamics - Bone: dyn_cape_l_03(ComponentPose=@n43, Alpha=1, ExternalForce="0,0,0") {
    angularDampingOverride=1
    angularSpringConstant=10000
    bAngularSpring=true
    bOverrideAngularDamping=true
    bUseGravityOverride=true
    gravityScale=0.35
    alphaBoolBlend {blendOption=Linear}
    boundBone {boneName=dyn_cape_l_03}
    componentAppliedLinearAccClamp=<3000 3000 3000>
    componentLinearAccScale=<2 2 2>
    componentLinearVelScale=<2 2 2>
    gravityOverride {z=-980}
    physicsBodyDefinitions [{
      boundBone {boneName=dyn_cape_l_03}
      boxExtents=<10 40 40>
      constraintSetup {angularLimitsMax=<0 5 15>, angularLimitsMin=<0 -40 -15>, angularTarget=<1 0 0>}
      localJointOffset=<5 0 0>
     }]
    relativeSpaceBone {boneName=backpack_root}
   }
   n45 = Anim Dynamics - Bone: dyn_cape_l_05(ComponentPose=@n44, Alpha=1, ExternalForce="0,0,0") {
    angularDampingOverride=1
    angularSpringConstant=10000
    bAngularSpring=true
    bOverrideAngularDamping=true
    bUseGravityOverride=true
    gravityScale=0.35
    alphaBoolBlend {blendOption=Linear}
    boundBone {boneName=dyn_cape_l_05}
    componentAppliedLinearAccClamp=<3000 3000 3000>
    componentLinearAccScale=<2 2 2>
    componentLinearVelScale=<2 2 2>
    gravityOverride {z=-980}
    physicsBodyDefinitions [{
      boundBone {boneName=dyn_cape_l_05}
      boxExtents=<10 40 40>
      constraintSetup {angularLimitsMax=<0 5 15>, angularLimitsMin=<0 -40 -15>, angularTarget=<1 0 0>}
      localJointOffset=<5 0 0>
     }]
    relativeSpaceBone {boneName=backpack_root}
   }
   n46 = Anim Dynamics - Bone: dyn_cape_l_06(ComponentPose=@n45, Alpha=1, ExternalForce="0,0,0") {
    angularDampingOverride=1
    angularSpringConstant=10000
    bAngularSpring=true
    bOverrideAngularDamping=true
    bUseGravityOverride=true
    gravityScale=0.35
    alphaBoolBlend {blendOption=Linear}
    boundBone {boneName=dyn_cape_l_06}
    componentAppliedLinearAccClamp=<3000 3000 3000>
    componentLinearAccScale=<2 2 2>
    componentLinearVelScale=<2 2 2>
    gravityOverride {z=-980}
    physicsBodyDefinitions [{
      boundBone {boneName=dyn_cape_l_06}
      boxExtents=<10 40 40>
      constraintSetup {angularLimitsMax=<0 5 15>, angularLimitsMin=<0 -40 -15>, angularTarget=<1 0 0>}
      localJointOffset=<5 0 0>
     }]
    relativeSpaceBone {boneName=backpack_root}
   }
   Save cached pose 'animDyn Cache'(Pose=Component To Local(ComponentPose=@n46)) {cacheName=animDyn Cache}
   n47 = RigidBody!(
    ComponentPose=Local To Component(LocalPose=Use cached pose 'Start_Cache'())
    Alpha
    OverrideWorldGravity:Vector="0,0,0"
    ExternalForce
    SimSpaceSettings:Sim Space Settings Structure=(WorldAlpha=1,VelocityScaleZ=0.25,DampingAlpha=0.5,MaxLinearVelocity=800,MaxAngularVelocity=8,MaxLinearAcceleration=4000,MaxAngularAcceleration=40,ExternalLinearDragV=(X=0,Y=0,Z=0),ExternalLinearVelocity=(X=0,Y=0,Z=0),ExternalAngularVelocity=(X=0,Y=0,Z=0))
   ) -> (Pose) {
    bClampLinearTranslationLimitToRefPose=true
    bOverrideWorldGravity=true
    overridePhysicsAsset=&/M_MED_HornetBerm_CapePhysics
    simulationSpace=BaseBoneSpace
    alphaBoolBlend {blendOption=Linear}
    baseBoneRef {boneName=spine_04}
    overrideWorldGravity {z=-980}
    simSpaceSettings {
     dampingAlpha=0.5
     maxAngularAcceleration=40
     maxAngularVelocity=8
     maxLinearAcceleration=4000
     maxLinearVelocity=800
     velocityScaleZ=0.25
     worldAlpha=1
    }
   }
   n48 = Component To Local(ComponentPose=@n47)
   n49 = RigidBody(
    ComponentPose=Local To Component(LocalPose=Use cached pose 'Start_Cache'())
    OverrideWorldGravity="0,0,0"
    SimSpaceSettings=(WorldAlpha=1,VelocityScaleZ=0.25,DampingAlpha=0.5,MaxLinearVelocity=800,MaxAngularVelocity=8,MaxLinearAcceleration=4000,MaxAngularAcceleration=40,ExternalLinearDragV=(X=0,Y=0,Z=0),ExternalLinearVelocity=(X=0,Y=0,Z=0),ExternalAngularVelocity=(X=0,Y=0,Z=0))
   ) {
    bClampLinearTranslationLimitToRefPose=true
    bOverrideWorldGravity=true
    overridePhysicsAsset=&/M_MED_HornetBerm_BackblingPin_CapePhysics
    simulationSpace=BaseBoneSpace
    alphaBoolBlend {blendOption=Linear}
    baseBoneRef {boneName=spine_04}
    overrideWorldGravity {z=-980}
    simSpaceSettings {
     dampingAlpha=0.5
     maxAngularAcceleration=40
     maxAngularVelocity=8
     maxLinearAcceleration=4000
     maxLinearVelocity=800
     velocityScaleZ=0.25
     worldAlpha=1
    }
   }
   n50 = Component To Local(ComponentPose=@n49)
   n51 = Get Is Wearing Backbling()
   Save cached pose 'RbanCache'(Pose=Blend Poses by bool(
     BlendPose_0=Use cached pose 'Start_Cache'()
     BlendPose_1=Blend Poses by bool(BlendPose_0=@n50, BlendPose_1=@n48, bActiveValue=@n51) {blendType=Linear}
     BlendTime_0=0.2
     BlendTime_1=0.5
     bActiveValue=Get bClothEnabled()
    ) {blendType=Linear}) {cacheName=RbanCache}
   n52 = Get TrailControllerRot() -> (TrailControllerRot:Rotator)
   n53 = "Transform (Modify) Bone - Bone: dyn_cape_l_01"(ComponentPose=Local To Component(LocalPose=Use cached pose 'Start_Cache'()), Alpha=1, Translation="0,0,0", Rotation=@n52, Scale="1,1,1") {rotationMode=BMM_Additive, rotationSpace=BCS_BoneSpace, alphaBoolBlend {blendOption=Linear}, boneToModify {boneName=dyn_cape_l_01}}
   n54 = "Transform (Modify) Bone - Bone: dyn_cape_01"(ComponentPose=@n53, Alpha=1, Translation="0,0,0", Rotation=@n52, Scale="1,1,1") {rotationMode=BMM_Additive, rotationSpace=BCS_BoneSpace, alphaBoolBlend {blendOption=Linear}, boneToModify {boneName=dyn_cape_01}}
   n55 = "Transform (Modify) Bone - Bone: dyn_cape_r_01"(ComponentPose=@n54, Alpha=1, Translation="0,0,0", Rotation=@n52, Scale="1,1,1") {rotationMode=BMM_Additive, rotationSpace=BCS_BoneSpace, alphaBoolBlend {blendOption=Linear}, boneToModify {boneName=dyn_cape_r_01}}
   n56 = Get BwdMovement()
   n57 = Trail controller - Bone: dyn_cape_l_07(ComponentPose=@n55, Alpha=@n56) {
    bLimitRotation=true
    bShowBaseMotion=false
    bShowLimit=false
    chainLength=7
    debugLifeTime=0.1
    maxDeltaTime=0.3333
    alphaScaleBiasClamp {bInterpResult=true, interpSpeedDecreasing=20, interpSpeedIncreasing=20}
    baseJoint {boneName=backpack_root}
    planarLimits [{drivingBone {boneName=backpack_root}, planeTransform {location=<0 0 0>, rotation=<rot 0 0 -90>, scale=<1 1 1>}}]
    relaxationSpeedScaleInputProcessor {bInterpResult=true, interpSpeedDecreasing=30, interpSpeedIncreasing=30}
    rotationLimits [
     {limitMax=<0 0 20>, limitMin=<0 -20 -20>}
     {limitMax=<0 0 20>, limitMin=<0 -20 -20>}
     {limitMax=<0 0 20>, limitMin=<0 -30 -20>}
     {limitMax=<0 5 20>, limitMin=<0 -45 -20>}
     {limitMax=<0 5 20>, limitMin=<0 -45 -20>}
     {limitMax=<0 0 20>, limitMin=<0 -45 -20>}
     {limitMax=<0 0 20>, limitMin=<0 -45 -20>}
    ]
    rotationOffsets [<0 0 0>, <0 0 0>, <0 0 0>, <0 0 0>, <0 0 0>, <0 0 0>, <0 0 0>]
    trailBone {boneName=dyn_cape_l_07}
    trailRelaxationSpeed {editorCurveData {keys [<at 0 23 RCIM_Linear>,<at 0.284 19.3799 RCIM_Linear>,<at 1 5 RCIM_Linear>]}}
   }
   n58 = Trail controller - Bone: dyn_cape_07(ComponentPose=@n57, Alpha=@n56) {
    bLimitRotation=true
    bShowBaseMotion=false
    bShowLimit=false
    chainLength=7
    debugLifeTime=0.1
    maxDeltaTime=0.3333
    alphaScaleBiasClamp {bInterpResult=true, interpSpeedDecreasing=20, interpSpeedIncreasing=20}
    baseJoint {boneName=backpack_root}
    planarLimits [{drivingBone {boneName=backpack_root}, planeTransform {location=<0 0 0>, rotation=<rot 0 0 -90>, scale=<1 1 1>}}]
    relaxationSpeedScaleInputProcessor {bInterpResult=true, interpSpeedDecreasing=30, interpSpeedIncreasing=30}
    rotationLimits [
     {limitMax=<0 0 20>, limitMin=<0 -20 -20>}
     {limitMax=<0 0 20>, limitMin=<0 -20 -20>}
     {limitMax=<0 0 20>, limitMin=<0 -30 -20>}
     {limitMax=<0 5 20>, limitMin=<0 -45 -20>}
     {limitMax=<0 5 20>, limitMin=<0 -45 -20>}
     {limitMax=<0 0 20>, limitMin=<0 -45 -20>}
     {limitMax=<0 0 20>, limitMin=<0 -45 -20>}
    ]
    rotationOffsets [<0 0 0>, <0 0 0>, <0 0 0>, <0 0 0>, <0 0 0>, <0 0 0>, <0 0 0>]
    trailBone {boneName=dyn_cape_07}
    trailRelaxationSpeed {editorCurveData {keys [<at 0 23 RCIM_Linear>,<at 0.284 19.3799 RCIM_Linear>,<at 1 5 RCIM_Linear>]}}
   }
   n59 = Trail controller - Bone: dyn_cape_r_07(ComponentPose=@n58, Alpha=@n56) {
    bLimitRotation=true
    bShowBaseMotion=false
    bShowLimit=false
    chainLength=7
    debugLifeTime=0.1
    maxDeltaTime=0.3333
    alphaScaleBiasClamp {bInterpResult=true, interpSpeedDecreasing=20, interpSpeedIncreasing=20}
    baseJoint {boneName=backpack_root}
    planarLimits [{drivingBone {boneName=backpack_root}, planeTransform {location=<0 0 0>, rotation=<rot 0 0 -90>, scale=<1 1 1>}}]
    relaxationSpeedScaleInputProcessor {bInterpResult=true, interpSpeedDecreasing=30, interpSpeedIncreasing=30}
    rotationLimits [
     {limitMax=<0 0 20>, limitMin=<0 -20 -20>}
     {limitMax=<0 0 20>, limitMin=<0 -20 -20>}
     {limitMax=<0 0 20>, limitMin=<0 -30 -20>}
     {limitMax=<0 5 20>, limitMin=<0 -45 -20>}
     {limitMax=<0 5 20>, limitMin=<0 -45 -20>}
     {limitMax=<0 0 20>, limitMin=<0 -45 -20>}
     {limitMax=<0 0 20>, limitMin=<0 -45 -20>}
    ]
    rotationOffsets [<0 0 0>, <0 0 0>, <0 0 0>, <0 0 0>, <0 0 0>, <0 0 0>, <0 0 0>]
    trailBone {boneName=dyn_cape_r_07}
    trailRelaxationSpeed {editorCurveData {keys [<at 0 23 RCIM_Linear>,<at 0.284 19.3799 RCIM_Linear>,<at 1 5 RCIM_Linear>]}}
   }
   Save cached pose 'trailCache'(Pose=Component To Local(ComponentPose=@n59)) {cacheName=trailCache}
  ]}
 EventGraph {nodes [
   Event OnCharacterCustomizationLoaded()
   > n1 = Sequence()
   n2 = CheckForBackbling()
   n3 = OR Boolean(A=@n2, B=@n2.bIsBackblingLong)
   > Set Is Wearing Backbling(Is Wearing Backbling=@n3)
   > Set Is Not Wearing Backbling(Is Not Wearing Backbling:bool=NOT Boolean(A=@n3)) -> (Output_Get:bool)
   n4 = Get Dynamic Material Instance(execute=@n1.then_1, SkeletalMeshComponent=GetPartSkeletalMeshComponent())
   > n5 = Branch(Condition=PickaxeItemDefinitionHasAnyMetaTag(Tags=Make Array("[0]"=(TagName="Cosmetics.MetaTags.HornetBerm")), bRequiresEquipped=true))
   > SetScalarParameterValue(self=@n4, ParameterName=Hide Element 01, Value=0)
   SetScalarParameterValue(execute=@n5.else, self=@n4, ParameterName=Hide Element 01, Value=1)
   Event OnBeginSkydiving()
   > Set TrailControllerRot(TrailControllerRot:Rotator="-65,0,0") -> (Output_Get:Rotator)
   n6 = Event BlueprintUpdateAnimation()
   > n7 = Sequence!(execute) -> (then_0, then_1, then_2:Exec)
   n8 = Get bIsPlayingEmote()
   > Set Is not Playing Emote Alpha(Is not Playing Emote Alpha:double="To Float (Boolean)"(InBool=NOT Boolean(A=@n8))) -> (Output_Get:double)
   n9 = Sequence!(execute=@n7.then_1) -> (then_0, then_1, then_2, then_3:Exec, then_4:Exec)
   > n10 = Cast To FortPlayerAnimInstance(Object=GetAnimInstance(self=GetOwnerSkeletalMeshComponent()))
   > Set isPistol(isPistol=false)
   > Set isRifle(isRifle:bool=false) -> (Output_Get:bool)
   > n11 = Switch on EFortWeaponCoreAnimation!(execute, Selection:EFortWeaponCoreAnimation Enum=Get WeaponCoreAnim(self=@n10)) -> (
    Melee:Exec
    Pistol:Exec
    Shotgun:Exec
    PaperBlueprint:Exec
    Rifle:Exec
    MeleeOneHand:Exec
    MachinePistol:Exec
    RocketLauncher:Exec
    GrenadeLauncher:Exec
    GoingCommando:Exec
    AssaultRifle:Exec
    TacticalShotgun:Exec
    SniperRifle:Exec
    TrapPlacement:Exec
    ShoulderLauncher:Exec
    AbilityDecoTool:Exec
    Crossbow:Exec
    C4:Exec
    RemoteControl:Exec
    DualWield:Exec
    AR_BullPup:Exec
    AR_ForwardGrip:Exec
    MedPackPaddles:Exec
    SMG_P90:Exec
    AR_DrumGun:Exec
    Consumable_Small:Exec
    Consumable_Large:Exec
    Balloon:Exec
    MountedTurret:Exec
    CreativeTool:Exec
    ExplosiveBow:Exec
    AshtonIndigo:Exec
    AshtonChicago:Exec
    MeleeDualWield:Exec
    Unarmed:Exec
   )
   Set isRifle(execute=[@n11.AssaultRifle, @n11.Rifle, @n11.RocketLauncher, @n11.Shotgun, @n11.SniperRifle, @n11.TacticalShotgun], isRifle=true)
   Set isPistol(execute=@n11.Pistol, isPistol=true)
   Set Is Rifle Alpha(execute=@n9.then_1, Is Rifle Alpha:double="To Float (Boolean)"(InBool=Get isRifle() -> (isRifle:bool))) -> (Output_Get:double)
   Set Is Pistol Alpha(execute=@n9.then_2, Is Pistol Alpha:double="To Float (Boolean)"(InBool=Get isPistol() -> (isPistol:bool))) -> (Output_Get:double)
   Set DeltaTimeLocal(execute=@n7.then_2, DeltaTimeLocal:double=@n6.DeltaTimeX) -> (Output_Get:double)
   > n12 = Sequence!(execute) -> (then_0, then_1, then_2)
   > n13 = Branch(Condition=Get bClothEnabled())
   > n14 = "Cloth Controller Function (Parameter Asset) (Nativized)"!(
    execute
    CharacterPartAnimInstance:Custom Character Part Anim Instance Object Reference=Self-Reference()
    ClothingAssetName:Name=Base_Cape_01
    ClothCoreSettings:"Cloth Core Settings Structure (by ref)"=Get Cloth Core Settings_Native() -> (Cloth Core Settings_Native:Cloth Core Settings Structure)
    WindGustLocalSettings:"Wind Gust Settings Structure (by ref)"=Get Wind Gust Local Settings Native() -> (Wind Gust Local Settings Native:Wind Gust Settings Structure)
    ClothUpAxis:Vector="1,0,0"
    LinVelScaleFwdBwdClamp:float=0
    LinVelScaleUpDwnClamp:float=0.7
    DeltaTime:float=Get DeltaTimeLocal() -> (DeltaTimeLocal:double)
    GameTimeInSeconds:float=GetGameTimeInSeconds()
    StateLogicLayerName:Name=none
    Parameters:Fort Character Dynamics Parameters Object Reference=&/M_MED_HornetBerm_Parameters
   ) -> (
    then
    OutCurrentState:Name=None
    OutClothParameters:Cloth Parameters Structure
    OutClothCoreSettings:Cloth Core Settings Structure
    OutWindGustSettings:Wind Gust Parameters Structure
    OutWindGustLocalSettings:Wind Gust Settings Structure
    OutWindVelocity:Vector
    OutLinearVelVector:Vector
   )
   > Set Wind Gust Local Settings Out(Wind Gust Local Settings Out:Wind Gust Settings Structure=@n14.OutWindGustLocalSettings) -> (Output_Get:Wind Gust Settings Structure)
   n15 = "Gravity Override State Function (Parameter Asset) (Nativized)"(
    execute=@n13.else
    CharacterPartAnimInstance=Self-Reference()
    GameTimeInSeconds=GetGameTimeInSeconds()
    Parameters=&M_MED_HornetBerm_Parameters
   )
   > Set GravityOverride(GravityOverride=@n15)
   > Set RBAN Settings Out(RBAN Settings Out=@n15.OutRigidBodyAnimNodeParameters)
   n16 = PlayerMovement(execute=@n12.then_1, AnimInstance=Self-Reference(), XDotProductCoefficient=1, YDotProductCoefficient=1, ZDotProductCoefficient=1)
   n17 = MapRangeClamped(Value=@n16.YMovement, InRangeA=0, InRangeB=1)
   n18 = "Absolute (Float)"(A=@n16)
   > Set BwdMovement(BwdMovement:double=MapRangeClamped(Value=float + float(A=@n18, B=@n17) {is=title K2Node_CommutativeAssociativeBinaryOperator}, InRangeA=0, InRangeB=1.4)) -> (Output_Get:double)
   n19 = Branch(execute=@n12.then_2, Condition=Get bIsSwimSprinting() -> (bIsSwimSprinting:bool))
   > Set TrailControllerRot(TrailControllerRot="15,0, 0")
   n20 = Branch(execute=@n19.else, Condition=Get bIsSkyDiving())
   > n21 = Branch(Condition=Get bIsParachuteOpened())
   > Set TrailControllerRot(TrailControllerRot="45,0, 0")
   n22 = Branch(execute=@n21.else, Condition="float > float"(A=Get PawnVelocityZ(), B=50) {is=title K2Node_CallFunction})
   > Set BwdMovement(BwdMovement=float * float(A=Get BwdMovement(), B=0))
   > Set TrailControllerRot(TrailControllerRot="45,0, 0")
   n23 = Branch(execute=@n22.else, Condition=Get bIsSkydiveDiveMode())
   > Set TrailControllerRot(TrailControllerRot="50,0,0")
   Set TrailControllerRot(execute=@n23.else, TrailControllerRot="60,0, 0")
   Set TrailControllerRot(execute=@n20.else, TrailControllerRot="15,0, 0")
   > Set BwdMovement(BwdMovement=float * float(A=Get BwdMovement(), B=1))
   Event OnEndSkydiving()
   > Set TrailControllerRot(TrailControllerRot="-10,0,0")
  ]}
]
variables [
 BwdMovement {type=number}
 Cloth Core Settings_Native {
  type=ClothCoreSettings
  default {
   collisionThickness=0
   dampingCoefficient=0.05
   frictionCoefficient=0.3
   localDampingCoefficient=0
   numberIterations=2
   numberSubsteps=1
   selfCollisionThickness=0
   animDriveDamping=<0 1>
   animDriveStiffness=<0 1>
   areaStiffness=<1 1>
   bendingStiffness=<1 1>
   edgeStiffness=<1 1>
   tetherStiffness=<1 1>
  }
 }
 DeltaTimeLocal {type=number, default=0.08}
 GravityOverride {type=Vector}
 Is not Playing Emote Alpha {type=number, default=1}
 Is Not Wearing Backbling {type=boolean}
 Is Pistol Alpha {type=number}
 Is Rifle Alpha {type=number}
 Is Wearing Backbling {type=boolean}
 isPistol {type=boolean}
 isRifle {type=boolean}
 RBAN Settings Out {
  type=FortRigidBodyAnimNodeParameters
  default {simSpaceSettings {
    maxAngularAcceleration=40
    maxAngularVelocity=8
    maxLinearAcceleration=4000
    maxLinearVelocity=800
    velocityScaleZ=0.5
    worldAlpha=1
    externalLinearDragV=<2 2 2>
   }}
 }
 TrailControllerRot {type=Rotator, default=<rot 15 0 0>}
 Wind Gust Local Settings Native {type=WindGustSettings, default {elapsedTime=0, finalGustStrengthLocal=0, gustDuration=0, gustStrength=0, timeBetweenGusts=0}}
 Wind Gust Local Settings Out {type=WindGustSettings, default {elapsedTime=0, finalGustStrengthLocal=0, gustDuration=0, gustStrength=0, timeBetweenGusts=0}}
]
