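# Physics asset of collision shapes for the M_MED_HornetBerm male medium player body cosmetic, named as its cloth colliders.
# Capsules cover the pelvis, spine, neck and clavicles, with a sphere for the head, while the arm, leg and foot bodies hold no shapes.
# Extra bodies add backpack and sheath capsules plus a broad, flat ground box.
# It defines no constraints, only per-bone shapes.
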PhysicsAsset /BRCosmetics/Characters/Player/Male/Medium/Bodies/M_MED_HornetBerm/Meshes/M_MED_HornetBerm_ClothColliders
bodies {
 backpack_collision [
  {shapeType=Capsule, length=1, radius=45, shapeName=backpack_collision_capsule, center=<0 -49.3919 -4.1251>, rotation=<rot 0 0 10>}
  {shapeType=Capsule, length=20, radius=20, shapeName=backpack_collision_capsule, center=<0 -24.6316 -18.953>, rotation=<rot 85 90 90>}
 ]
 calf_l []
 calf_r []
 clavicle_l [{shapeType=Capsule, length=9.0192, radius=10, center=<9.3294 -0.5 -1.5>, rotation=<rot 79.3028 -1.4093 174.1852>}]
 clavicle_r [{shapeType=Capsule, length=9.0192, radius=10, center=<-9.3294 0.5 1.5>, rotation=<rot 79.3028 -1.4093 -5.8148>}]
 dyn_sheath_01 [{shapeType=Capsule, length=62, radius=4.5, shapeName=dyn_sheath_01_capsule, center=<0 0 -17.7033>, rotation=<rot 1.8963 -90 2.438>}]
 dyn_sheath_MSTR [{shapeType=Capsule, length=0.5, radius=0.5, shapeName=dyn_sheath_MSTR_capsule}]
 foot_l []
 foot_r []
 ground_collision [{shapeType=Box, extentX=250, extentY=250, extentZ=20, shapeName=ground_collision_box, center=<0 0 -10>, rotation=<rot 0 -90 0>}]
 head [{shapeType=Sphere, radius=15, center=<7.6055 4.3229 0>}]
 lowerarm_l []
 lowerarm_r []
 neck_02 [{shapeType=Capsule, length=13.4375, radius=7.3175, center=<-0.3402 0.6739 0>, rotation=<rot 90 4.2785 0>}]
 pelvis [
  {shapeType=Capsule, length=10, radius=14, center=<-2.4887 -0.2961 0>}
  {shapeType=Capsule, length=15, radius=14, shapeName=pelvis_capsule, center=<-16.7538 0.0499 0>}
 ]
 spine_01 [{shapeType=Capsule, length=7.9438, radius=12, center=<-1.7965 -0.1001 0>}]
 spine_02 [
  {shapeType=Capsule, length=8, radius=12, center=<-1.711 0.9076 0.0001>}
  {shapeType=Capsule, length=29.3956, radius=11.5807, center=<2.5024 0.4992 3.9825>, rotation=<rot 90 0.0049 0.0049>}
  {shapeType=Capsule, length=29.875, radius=11.1013, center=<2.4704 0.4991 -3.9825>, rotation=<rot 90 -0.0396 -0.0396>}
 ]
 spine_03 [{shapeType=Capsule, length=10, radius=12, center=<0.6754 1.1909 0.0001>}]
 spine_04 [{shapeType=Capsule, length=9.4404, radius=13, center=<3.2671 0.2445 0.0002>}]
 spine_05 [
  {shapeType=Capsule, length=10, radius=13, center=<-4.1001 -0.0008 0>}
  {shapeType=Capsule, length=13.9518, radius=13, center=<-0.6283 -0.2182 0>}
 ]
 thigh_l []
 thigh_r []
 upperarm_l []
 upperarm_r []
}
constraints []
